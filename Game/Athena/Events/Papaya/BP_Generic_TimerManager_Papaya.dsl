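; Event countdown manager that shows time left until a calendar event ends, outside Creative and Playground playlists.
; At match ready on authority it watches the event and sets ServerWorldSecondsAtEvent, whose OnRep starts periodic calendar resyncs on authority and a one-second timer on remotes.
; Each tick splits days, hours, minutes and seconds into digits by string reversal, writing them to one of two Material Parameter Collections.
; At zero it shows final graphics, then clears them.
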(event Collision|EventActorBeginOverlap
  (params (OtherActor :type Actor%20Object%20Reference)))

(event EventTick
  (params (DeltaSeconds :type Float%20%28single-precision%29)))

(event EventBeginPlay
  (bind _gamestate (Class|FortAsyncActionCurrentPlaylistReady|PlaylistDataReadyAsync)
    (:OnReady
      (if (not (Class|FortGameStateAthena|IsCreativeorPlaygroundModePlaylist _gamestate))
        (Class|FortAsyncActionAuthorityMatchReady|MatchReadyAsync
          (:OnReady
            (bind (_eventname _timeuntilend) (MCPContext|WatchCalendarEvent (Variables|Default|GetCalendarEventName))
              (:OnEventActive
                (Variables|Default|SetTimeUntilCountdownEnd _timeuntilend)
                (CallFunction|ContinueClockCountdown))) ; Leadup to Festivus
          ))))))

(event Custom|ContinueClockCountdown
  (Variables|Default|SetServerWorldSecondsAtEvent (+ (GameState|GetServerWorldTimeSeconds (Game|GetGameState)) (Math|Timespan|GetTotalSeconds (Variables|Default|GetTimeUntilCountdownEnd)))))

(event Custom|CountdownTimer_New
  (bind _returnvalue (Game|GetGameState))
  (bind _returnvalue_2 (GameState|GetServerWorldTimeSeconds _returnvalue))
  (bind _serverworldsecondsatevent (Variables|Default|GetServerWorldSecondsAtEvent))
  (bind _returnvalue_3 (- _serverworldsecondsatevent _returnvalue_2))
  (bind _returnvalue_4 (Math|Float|Truncate _returnvalue_3))
  (bind _timeoffsetseconds (Variables|Default|GetTimeOffsetSeconds))
  (bind _returnvalue_5 (+ _returnvalue_4 _timeoffsetseconds))
  (bind _output (Utilities|ClampMinInteger _returnvalue_5))
  (Development|PrintString (Utilities|String|ToString(Integer) _output))
  (bind _output_get (Variables|Default|SetSecondsRemaining _output))
  (CallFunction|VisualUpdateEverySecond)
  ;[1 1 1] count down is over show final graphic
  (if (<= _output_get 0)
    (Utilities|FlowControl|DoOnce false
      (:Completed
        (Utilities|Time|ClearandInvalidateTimerbyHandle (Variables|Default|GetCountdownTimerHandle))
        (if (Variables|Default|GetIs_2nd_Countdown)
          ;[1 1 1] 2nd countdown
          (Rendering|Material|SetScalarParameterValue "/Game/Athena/Environments/Blueprints/Events/MPC_Countdown_2nd.MPC_Countdown_2nd" "ShowCountdown" 1.0 :!class /Script/Engine.KismetMaterialLibrary)
          (Rendering|Material|SetScalarParameterValue "/Game/Athena/Environments/Blueprints/Events/MPC_Countdown_2nd.MPC_Countdown_2nd" "ShowCountdownPhase2" 1.0 :!class /Script/Engine.KismetMaterialLibrary) ; Shows 2nd stage of countdown (hologram countdown floating in the sky)
          (Rendering|Material|SetScalarParameterValue "/Game/Athena/Environments/Blueprints/Events/MPC_Countdown_2nd.MPC_Countdown_2nd" "ShowCountdownPhase3" 1.0 :!class /Script/Engine.KismetMaterialLibrary) ; Shows post countdown graphics
          (else
            (Rendering|Material|SetScalarParameterValue "/Game/Athena/Events/Generic/CountdownTimer/Materials/MPC_Countdown.MPC_Countdown" "ShowCountdown" 1.0 :!class /Script/Engine.KismetMaterialLibrary)
            (Rendering|Material|SetScalarParameterValue "/Game/Athena/Events/Generic/CountdownTimer/Materials/MPC_Countdown.MPC_Countdown" "ShowCountdownPhase2" 1.0 :!class /Script/Engine.KismetMaterialLibrary) ; Shows 2nd stage of countdown (hologram countdown floating in the sky)
            (Rendering|Material|SetScalarParameterValue "/Game/Athena/Events/Generic/CountdownTimer/Materials/MPC_Countdown.MPC_Countdown" "ShowCountdownPhase3" 1.0 :!class /Script/Engine.KismetMaterialLibrary) ; Shows post countdown graphics
          ))
        (Utilities|Time|SetTimerbyEvent (delegate Custom|ClearTimerVisuals) (Variables|Default|GetPhase3VisualDuration))))))

(event Custom|ClearTimerVisuals
  (if (Variables|Default|GetIs_2nd_Countdown)
    (Rendering|Material|SetScalarParameterValue "/Game/Athena/Environments/Blueprints/Events/MPC_Countdown_2nd.MPC_Countdown_2nd" "ShowCountdown" :!class /Script/Engine.KismetMaterialLibrary)
    (Rendering|Material|SetScalarParameterValue "/Game/Athena/Environments/Blueprints/Events/MPC_Countdown_2nd.MPC_Countdown_2nd" "ShowCountdownPhase2" :!class /Script/Engine.KismetMaterialLibrary) ; Shows 2nd stage of countdown (hologram countdown floating in the sky)
    (Rendering|Material|SetScalarParameterValue "/Game/Athena/Environments/Blueprints/Events/MPC_Countdown_2nd.MPC_Countdown_2nd" "ShowCountdownPhase3" :!class /Script/Engine.KismetMaterialLibrary) ; Shows post countdown graphics
    (else
      (Rendering|Material|SetScalarParameterValue "/Game/Athena/Events/Generic/CountdownTimer/Materials/MPC_Countdown.MPC_Countdown" "ShowCountdown" :!class /Script/Engine.KismetMaterialLibrary)
      (Rendering|Material|SetScalarParameterValue "/Game/Athena/Events/Generic/CountdownTimer/Materials/MPC_Countdown.MPC_Countdown" "ShowCountdownPhase2" :!class /Script/Engine.KismetMaterialLibrary) ; Shows 2nd stage of countdown (hologram countdown floating in the sky)
      (Rendering|Material|SetScalarParameterValue "/Game/Athena/Events/Generic/CountdownTimer/Materials/MPC_Countdown.MPC_Countdown" "ShowCountdownPhase3" :!class /Script/Engine.KismetMaterialLibrary) ; Shows post countdown graphics
    )))

(event Custom|ErrorCorrection
  ;[1 1 1] Error correction for timer
  (Utilities|FlowControl|DoOnce false
    (:Completed
      (Utilities|Time|SetTimerbyEvent (EventDispatchers|CreateEvent :!event Custom|StartFinalAccuracyadjustment) (Math|Float|Clamp(Float) (Math|Conversions|ToFloat(Integer) (- (Utilities|ClampMinInteger (+ (Math|Float|Truncate (- (Variables|Default|GetServerWorldSecondsAtEvent) (GameState|GetServerWorldTimeSeconds (Game|GetGameState)))) (Variables|Default|GetTimeOffsetSeconds))) 300)) 0.001 1e+10)))))

(event Custom|StartFinalAccuracyadjustment
  ;[1 1 1] Start the accuracy adjustment
  (bind _returnvalue (Utilities|Time|SetTimerbyEvent (EventDispatchers|CreateEvent :!event Custom|CountDownTimer_AccuracyFix) 3e+01 true))
  (Variables|Default|SetCountDownTimerAccuracyfix _returnvalue))

(event Custom|CountDownTimer_AccuracyFix
  ;[1 1 1] Fix Accuracy every once in a wile based on timer ( more often in the last minutes)
  (bind (_biseventactive _timesincebegin _timeuntilend) (CalendarEvents|GetCalendarEventInformation (Variables|Default|GetCalendarEventName)))
  (if _biseventactive
    (Variables|Default|SetTimeUntilCountdownEnd _timeuntilend)
    (Variables|Default|SetServerWorldSecondsAtEvent (+ (GameState|GetServerWorldTimeSeconds (Game|GetGameState)) (Math|Timespan|GetTotalSeconds (Variables|Default|GetTimeUntilCountdownEnd))))))

(event Custom|VisualUpdateEverySecond
  (bind _returnvalue (Math|Timespan|GetMinutes (Variables|Default|GetTimespanRemaining)))
  (bind _returnvalue_2 (Utilities|String|ToString(Integer) _returnvalue))
  (bind _returnvalue_3 (Utilities|String|Reverse _returnvalue_2))
  (bind _returnvalue_4 (Utilities|String|GetSubstring _returnvalue_3 1))
  (bind _returnvalue_5 (Math|Timespan|GetSeconds (Variables|Default|GetTimespanRemaining)))
  (bind _returnvalue_6 (Utilities|String|ToString(Integer) _returnvalue_5))
  (bind _returnvalue_7 (Math|Timespan|GetHours (Variables|Default|GetTimespanRemaining)))
  (bind _returnvalue_8 (Utilities|String|ToString(Integer) _returnvalue_7))
  (bind _returnvalue_9 (Utilities|String|Reverse _returnvalue_8))
  (bind _returnvalue_10 (Utilities|String|GetSubstring _returnvalue_9))
  (bind _returnvalue_11 (Utilities|String|StringToFloat _returnvalue_10))
  (bind _returnvalue_12 (Utilities|String|GetSubstring _returnvalue_9 1))
  (bind _returnvalue_13 (Utilities|String|StringToFloat _returnvalue_12))
  (bind _returnvalue_14 (Utilities|String|Reverse _returnvalue_6))
  (bind _returnvalue_15 (Utilities|String|GetSubstring _returnvalue_14 1))
  (bind _returnvalue_16 (Utilities|String|StringToFloat _returnvalue_15))
  (bind _returnvalue_17 (Utilities|String|GetSubstring _returnvalue_14))
  (bind _returnvalue_18 (Utilities|String|StringToFloat _returnvalue_17))
  (bind _returnvalue_19 (Math|Timespan|GetDays (Variables|Default|GetTimespanRemaining)))
  (bind _returnvalue_20 (Utilities|String|ToString(Integer) _returnvalue_19))
  (bind _returnvalue_21 (Utilities|String|Reverse _returnvalue_20))
  (bind _returnvalue_22 (Utilities|String|GetSubstring _returnvalue_21))
  (bind _returnvalue_23 (Utilities|String|StringToFloat _returnvalue_22))
  (bind _returnvalue_24 (Utilities|String|GetSubstring _returnvalue_21 1))
  (bind _returnvalue_25 (Utilities|String|StringToFloat _returnvalue_24))
  (bind _returnvalue_26 (Utilities|String|StringToFloat _returnvalue_4))
  (bind _returnvalue_27 (Utilities|String|GetSubstring _returnvalue_3))
  (bind _returnvalue_28 (Utilities|String|StringToFloat _returnvalue_27))
  (bind _secondsremaining (Variables|Default|GetSecondsRemaining))
  (bind _timespanremaining (Variables|Default|GetTimespanRemaining))
  ;[1 1 1] Visual Update
  (if (<= _secondsremaining 0)
    (if (Variables|Default|GetIs_2nd_Countdown)
      ;;[1 1 1] 2nd countdown to have two countdown in one map
      (Rendering|Material|SetScalarParameterValue "/Game/Athena/Environments/Blueprints/Events/MPC_Countdown_2nd.MPC_Countdown_2nd" "ShowCountdown" 1.0 :!class /Script/Engine.KismetMaterialLibrary)
      (Rendering|Material|SetScalarParameterValue "/Game/Athena/Environments/Blueprints/Events/MPC_Countdown_2nd.MPC_Countdown_2nd" "ShowCountdownPhase2" 1.0 :!class /Script/Engine.KismetMaterialLibrary) ; Shows 2nd stage of countdown (hologram countdown floating in the sky)
      (Rendering|Material|SetScalarParameterValue "/Game/Athena/Environments/Blueprints/Events/MPC_Countdown_2nd.MPC_Countdown_2nd" "ShowCountdownPhase3" 1.0 :!class /Script/Engine.KismetMaterialLibrary) ; Shows post countdown graphics
      (Rendering|Material|SetScalarParameterValue "/Game/Athena/Environments/Blueprints/Events/MPC_Countdown_2nd.MPC_Countdown_2nd" "Digit1" :!class /Script/Engine.KismetMaterialLibrary)
      (Rendering|Material|SetScalarParameterValue "/Game/Athena/Environments/Blueprints/Events/MPC_Countdown_2nd.MPC_Countdown_2nd" "Digit2" :!class /Script/Engine.KismetMaterialLibrary)
      (Rendering|Material|SetScalarParameterValue "/Game/Athena/Environments/Blueprints/Events/MPC_Countdown_2nd.MPC_Countdown_2nd" "Digit3" :!class /Script/Engine.KismetMaterialLibrary)
      (Rendering|Material|SetScalarParameterValue "/Game/Athena/Environments/Blueprints/Events/MPC_Countdown_2nd.MPC_Countdown_2nd" "Digit4" :!class /Script/Engine.KismetMaterialLibrary)
      (Rendering|Material|SetScalarParameterValue "/Game/Athena/Environments/Blueprints/Events/MPC_Countdown_2nd.MPC_Countdown_2nd" "Digit5" :!class /Script/Engine.KismetMaterialLibrary)
      (Rendering|Material|SetScalarParameterValue "/Game/Athena/Environments/Blueprints/Events/MPC_Countdown_2nd.MPC_Countdown_2nd" "Digit6" :!class /Script/Engine.KismetMaterialLibrary)
      (Rendering|Material|SetScalarParameterValue "/Game/Athena/Environments/Blueprints/Events/MPC_Countdown_2nd.MPC_Countdown_2nd" "Digit7" :!class /Script/Engine.KismetMaterialLibrary)
      (Rendering|Material|SetScalarParameterValue "/Game/Athena/Environments/Blueprints/Events/MPC_Countdown_2nd.MPC_Countdown_2nd" "Digit8" :!class /Script/Engine.KismetMaterialLibrary)
      (else
        ;;[1 1 1] count down is over show final graphic
        (Rendering|Material|SetScalarParameterValue "/Game/Athena/Events/Generic/CountdownTimer/Materials/MPC_Countdown.MPC_Countdown" "ShowCountdown" 1.0 :!class /Script/Engine.KismetMaterialLibrary)
        (Rendering|Material|SetScalarParameterValue "/Game/Athena/Events/Generic/CountdownTimer/Materials/MPC_Countdown.MPC_Countdown" "ShowCountdownPhase2" 1.0 :!class /Script/Engine.KismetMaterialLibrary) ; Shows 2nd stage of countdown (hologram countdown floating in the sky)
        (Rendering|Material|SetScalarParameterValue "/Game/Athena/Events/Generic/CountdownTimer/Materials/MPC_Countdown.MPC_Countdown" "ShowCountdownPhase3" 1.0 :!class /Script/Engine.KismetMaterialLibrary) ; Shows post countdown graphics
        (Rendering|Material|SetScalarParameterValue "/Game/Athena/Events/Generic/CountdownTimer/Materials/MPC_Countdown.MPC_Countdown" "Digit1" :!class /Script/Engine.KismetMaterialLibrary)
        (Rendering|Material|SetScalarParameterValue "/Game/Athena/Events/Generic/CountdownTimer/Materials/MPC_Countdown.MPC_Countdown" "Digit2" :!class /Script/Engine.KismetMaterialLibrary)
        (Rendering|Material|SetScalarParameterValue "/Game/Athena/Events/Generic/CountdownTimer/Materials/MPC_Countdown.MPC_Countdown" "Digit3" :!class /Script/Engine.KismetMaterialLibrary)
        (Rendering|Material|SetScalarParameterValue "/Game/Athena/Events/Generic/CountdownTimer/Materials/MPC_Countdown.MPC_Countdown" "Digit4" :!class /Script/Engine.KismetMaterialLibrary)
        (Rendering|Material|SetScalarParameterValue "/Game/Athena/Events/Generic/CountdownTimer/Materials/MPC_Countdown.MPC_Countdown" "Digit5" :!class /Script/Engine.KismetMaterialLibrary)
        (Rendering|Material|SetScalarParameterValue "/Game/Athena/Events/Generic/CountdownTimer/Materials/MPC_Countdown.MPC_Countdown" "Digit6" :!class /Script/Engine.KismetMaterialLibrary)
        (Rendering|Material|SetScalarParameterValue "/Game/Athena/Events/Generic/CountdownTimer/Materials/MPC_Countdown.MPC_Countdown" "Digit7" :!class /Script/Engine.KismetMaterialLibrary)
        (Rendering|Material|SetScalarParameterValue "/Game/Athena/Events/Generic/CountdownTimer/Materials/MPC_Countdown.MPC_Countdown" "Digit8" :!class /Script/Engine.KismetMaterialLibrary)))
    (else
      (Variables|Default|SetTimespanRemaining (Math|Timespan|MakeTimespan 0 0 0 _secondsremaining))
      (if (Variables|Default|GetIs_2nd_Countdown)
        (Rendering|Material|SetScalarParameterValue "/Game/Athena/Environments/Blueprints/Events/MPC_Countdown_2nd.MPC_Countdown_2nd" "ShowCountdownPhase2" 1.0 :!class /Script/Engine.KismetMaterialLibrary)
        (Rendering|Material|SetScalarParameterValue "/Game/Athena/Environments/Blueprints/Events/MPC_Countdown_2nd.MPC_Countdown_2nd" "ShowCountdownPhase3" :!class /Script/Engine.KismetMaterialLibrary)
        (Rendering|Material|SetScalarParameterValue "/Game/Athena/Environments/Blueprints/Events/MPC_Countdown_2nd.MPC_Countdown_2nd" "ShowCountdown" 1.0 :!class /Script/Engine.KismetMaterialLibrary)
        (Rendering|Material|SetScalarParameterValue "/Game/Athena/Environments/Blueprints/Events/MPC_Countdown_2nd.MPC_Countdown_2nd" "Digit1" _returnvalue_18 :!class /Script/Engine.KismetMaterialLibrary)
        (Rendering|Material|SetScalarParameterValue "/Game/Athena/Environments/Blueprints/Events/MPC_Countdown_2nd.MPC_Countdown_2nd" "Digit2" _returnvalue_16 :!class /Script/Engine.KismetMaterialLibrary)
        (Rendering|Material|SetScalarParameterValue "/Game/Athena/Environments/Blueprints/Events/MPC_Countdown_2nd.MPC_Countdown_2nd" "Digit3" _returnvalue_28 :!class /Script/Engine.KismetMaterialLibrary)
        (Rendering|Material|SetScalarParameterValue "/Game/Athena/Environments/Blueprints/Events/MPC_Countdown_2nd.MPC_Countdown_2nd" "Digit4" _returnvalue_26 :!class /Script/Engine.KismetMaterialLibrary)
        (Rendering|Material|SetScalarParameterValue "/Game/Athena/Environments/Blueprints/Events/MPC_Countdown_2nd.MPC_Countdown_2nd" "Digit5" _returnvalue_11 :!class /Script/Engine.KismetMaterialLibrary)
        (Rendering|Material|SetScalarParameterValue "/Game/Athena/Environments/Blueprints/Events/MPC_Countdown_2nd.MPC_Countdown_2nd" "Digit6" _returnvalue_13 :!class /Script/Engine.KismetMaterialLibrary)
        (Rendering|Material|SetScalarParameterValue "/Game/Athena/Environments/Blueprints/Events/MPC_Countdown_2nd.MPC_Countdown_2nd" "Digit7" _returnvalue_23 :!class /Script/Engine.KismetMaterialLibrary)
        (Rendering|Material|SetScalarParameterValue "/Game/Athena/Environments/Blueprints/Events/MPC_Countdown_2nd.MPC_Countdown_2nd" "Digit8" _returnvalue_25 :!class /Script/Engine.KismetMaterialLibrary)
        (else
          (Rendering|Material|SetScalarParameterValue "/Game/Athena/Events/Generic/CountdownTimer/Materials/MPC_Countdown.MPC_Countdown" "ShowCountdownPhase2" 1.0 :!class /Script/Engine.KismetMaterialLibrary)
          (Rendering|Material|SetScalarParameterValue "/Game/Athena/Events/Generic/CountdownTimer/Materials/MPC_Countdown.MPC_Countdown" "ShowCountdownPhase3" :!class /Script/Engine.KismetMaterialLibrary)
          (Rendering|Material|SetScalarParameterValue "/Game/Athena/Events/Generic/CountdownTimer/Materials/MPC_Countdown.MPC_Countdown" "ShowCountdown" 1.0 :!class /Script/Engine.KismetMaterialLibrary)
          (Rendering|Material|SetScalarParameterValue "/Game/Athena/Events/Generic/CountdownTimer/Materials/MPC_Countdown.MPC_Countdown" "Digit1" _returnvalue_18 :!class /Script/Engine.KismetMaterialLibrary)
          (Rendering|Material|SetScalarParameterValue "/Game/Athena/Events/Generic/CountdownTimer/Materials/MPC_Countdown.MPC_Countdown" "Digit2" _returnvalue_16 :!class /Script/Engine.KismetMaterialLibrary)
          (Rendering|Material|SetScalarParameterValue "/Game/Athena/Events/Generic/CountdownTimer/Materials/MPC_Countdown.MPC_Countdown" "Digit3" _returnvalue_28 :!class /Script/Engine.KismetMaterialLibrary)
          (Rendering|Material|SetScalarParameterValue "/Game/Athena/Events/Generic/CountdownTimer/Materials/MPC_Countdown.MPC_Countdown" "Digit4" _returnvalue_26 :!class /Script/Engine.KismetMaterialLibrary)
          (Rendering|Material|SetScalarParameterValue "/Game/Athena/Events/Generic/CountdownTimer/Materials/MPC_Countdown.MPC_Countdown" "Digit5" _returnvalue_11 :!class /Script/Engine.KismetMaterialLibrary)
          (Rendering|Material|SetScalarParameterValue "/Game/Athena/Events/Generic/CountdownTimer/Materials/MPC_Countdown.MPC_Countdown" "Digit6" _returnvalue_13 :!class /Script/Engine.KismetMaterialLibrary)
          (Rendering|Material|SetScalarParameterValue "/Game/Athena/Events/Generic/CountdownTimer/Materials/MPC_Countdown.MPC_Countdown" "Digit7" _returnvalue_23 :!class /Script/Engine.KismetMaterialLibrary)
          (Rendering|Material|SetScalarParameterValue "/Game/Athena/Events/Generic/CountdownTimer/Materials/MPC_Countdown.MPC_Countdown" "Digit8" _returnvalue_25 :!class /Script/Engine.KismetMaterialLibrary))))))

(fn OnRep_ServerWorldSecondsAtEvent
  (if (<= (Variables|Default|GetServerWorldSecondsAtEvent) 0.0)
    (if (Variables|Default|GetIs_2nd_Countdown)
      (Rendering|Material|SetScalarParameterValue "/Game/Athena/Environments/Blueprints/Events/MPC_Countdown_2nd.MPC_Countdown_2nd" "ShowPhase2" :!class /Script/Engine.KismetMaterialLibrary)
      (else
        (Rendering|Material|SetScalarParameterValue "/Game/Athena/Events/Generic/CountdownTimer/Materials/MPC_Countdown.MPC_Countdown" "ShowPhase2" :!class /Script/Engine.KismetMaterialLibrary)))
    (else
      (Networking|SwitchHasAuthority
        (:Authority
          (CallFunction|StartFinalAccuracyadjustment))
        (:Remote
          (if (not (Utilities|Time|IsTimerActivebyHandle (Variables|Default|GetCountdownTimerHandle)))
            (Variables|Default|SetSecondsRemaining (Utilities|ClampMinInteger (+ (Math|Float|Truncate (- (Variables|Default|GetServerWorldSecondsAtEvent) (GameState|GetServerWorldTimeSeconds (Game|GetGameState)))) (Variables|Default|GetTimeOffsetSeconds))))
            (bind _returnvalue (Utilities|Time|SetTimerbyEvent (EventDispatchers|CreateEvent :!event CountdownTimer_New) 1.0 true))
            (Variables|Default|SetCountdownTimerHandle _returnvalue)))))))

(fn ConstructionScript)
